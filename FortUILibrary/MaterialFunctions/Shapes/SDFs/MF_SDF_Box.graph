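# UI material function that outputs a signed distance field for a box with optional rounded corners.
# UVs are centered and mirrored with abs, offset by Size, then optionally scaled by the aspect ratio from PixelSize to keep corners undistorted.
# MF_UI_IsolateQuads picks a CornerRoundness value per quadrant, allowing four different radii.
# The standard rounded-box formula sums inside and outside distances, then subtracts that radius.
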MaterialFunction /FortUILibrary/MaterialFunctions/Shapes/SDFs/MF_SDF_Box
graph {
 properties {
  bExposeToLibrary=true
  description=Signed distance field for a box with optional rounded corners.
  previewMaterialDomain=MD_UI
  libraryCategoriesText [SDFs]
 }
 nodes [
  n1 = FunctionInput(Preview=MaterialFunctionCall() {materialFunction=&GetUserInterfaceUV}.Pixel Size) {bUsePreviewValueAsDefault=true, description=Pixel Size Reference, inputName=PixelSize, inputType=FunctionInput_Vector2, sortPriority=100}
  n2 = MaterialFunctionCall!(Dimensions=@n1) {materialFunction=/FortUILibrary/MaterialFunctions/Basics/MF_UI_GetAspectRatio}
  n3 = Constant() {r=1}
  n4 = FunctionInput(Preview=StaticBool()) {
   bUsePreviewValueAsDefault=true
   description=Enable to maintain the aspect ratio of rounded corners as the width/height of the box is scaled in either direction.
   inputName=KeepAspectRatio
   inputType=FunctionInput_StaticBool
   sortPriority=20
   previewValue=<0 0 0 0>
  }
  n5 = FunctionInput(Preview=TextureCoordinate()) {
   bUsePreviewValueAsDefault=true
   description="Coordinates of the distance field (0,0 is origin)."
   inputName=UVs
   inputType=FunctionInput_Vector2
  }
  n6 = Abs(None=ConstantBiasScale(None=@n5) {bias=-0.5, scale=2})
  n7 = FunctionInput() {
   bUsePreviewValueAsDefault=true
   description="The size of the box. Use a Float (Scalar) for an equilateral square or a Float2 for a rectangle with separate Width/Height values."
   inputName=Size
   inputType=FunctionInput_Vector2
   sortPriority=10
   previewValue=<0.5 0.5 0 1>
  }
  n8 = FunctionInput() {
   bUsePreviewValueAsDefault=true
   description="The radius (usually between '0.0' and '0.5') of the rounded corners. Use a Float (Scalar) for uniform roundness, or a Float4 to set a different radius for each corner."
   inputName=CornerRoundness
   inputType=FunctionInput_Vector4
   sortPriority=15
   previewValue=<0 0 0 0>
  }
  n9 = MaterialFunctionCall!(UVs=@n5, Use Raw UVs, 4-Quad Values=@n8) {materialFunction=/FortUILibrary/MaterialFunctions/Basics/MF_UI_IsolateQuads}
  n10 = Add(A=Multiply(A=StaticSwitch(True=@n2.Max Ratio, False=@n3, Value=@n4), B=Subtract(A=@n6, B=@n7)), B=@n9.Isolated 4-Quad)
  n11 = Min(A=Max(A=ComponentMask(None=@n10) {r=true}, B=ComponentMask(None=@n10) {g=true})) {constB=0}
  n12 = Length(None=Max(A=@n10) {constB=0}) {is=class MaterialExpressionLength}
  FunctionOutput(None=Subtract(A=Add(A=@n11, B=@n12), B=@n9.Isolated 4-Quad)) {description=The resulting box SDF., outputName=SDF}
 ]
}
