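# UI material function that builds a progress bar shape from a box SDF, outputting an anti-aliased Fill mask and the raw SDF.
# UVs are scaled about their center and offset horizontally by half of an inverted scalar, with a static switch to flip direction.
# Inputs cover progress, width scale, height, corner roundness, edge threshold and reverse.
# A Step/Max branch meant to force empty and full at 0 and 1 feeds neither output.
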MaterialFunction /FortUILibrary/MaterialFunctions/Shapes/SDFs/MF_UI_SDF_Progress
graph {
 properties {bExposeToLibrary=true, previewMaterialDomain=MD_UI, libraryCategoriesText [UI]}
 nodes [
  n1 = NamedRerouteUsage()
  n2 = MaterialFunctionCall(X=NamedRerouteUsage(), Y=Constant() {r=1}) {materialFunction=&MakeFloat2}
  n3 = Divide(A=OneMinus(None=NamedRerouteUsage())) {constB=-2}
  n4 = StaticSwitch(True=Multiply(A=@n3) {constB=-1}, False=@n3, Value=NamedRerouteUsage())
  n5 = MaterialFunctionCall!(
   UVs=MaterialFunctionCall!(
    Input UVs=MaterialFunctionCall(UVs=@n1, Texture Scale=@n2.Result) {materialFunction=&ScaleUVsByCenter}.UVs
    Translate=MaterialFunctionCall(X=@n4) {materialFunction=&MakeFloat2}.Result
   ) {materialFunction=/FortUILibrary/MaterialFunctions/Basics/MF_UI_Translate}.UVs raw
   Size=MaterialFunctionCall(X=NamedRerouteUsage(), Y=NamedRerouteUsage()) {materialFunction=&MakeFloat2}.Result
   CornerRoundness=NamedRerouteUsage()
   KeepAspectRatio
   PixelSize
  ) {materialFunction=&/MF_SDF_Box}
  n6 = MaterialFunctionCall!(DistanceField=@n5.SDF, Threshold=NamedRerouteUsage()) {materialFunction=/FortUILibrary/MaterialFunctions/Shapes/Modifiers/MF_UI_AntiAliasing_DPI}
  FunctionOutput(None=@n6.Inner) {outputName=Fill}
  FunctionOutput(None=@n5.SDF) {outputName=SDF}
  n7 = NamedRerouteUsage()
  Max(
   A=Multiply(A=@n6.Inner, B=Step(X=@n7) {constY=0.001}) {desc="Ensure that 0 will completely \r\nmask the area, regardless of threshold."}
   B=Step(X=@n7) {constY=1}
  ) {desc="Ensure that 1 will completely \r\nshow the area, regadless of \r\nthreshold."}
  NamedRerouteDeclaration(None=FunctionInput(Preview=Constant() {r=0.01}) {bUsePreviewValueAsDefault=true, inputName=ProgressEdgeThreshold, inputType=FunctionInput_Scalar, sortPriority=1}) {name=ProgressEdgeThreshold, nodeColor=<rgba 0 0.4583 0.0978 1>}
  NamedRerouteDeclaration(None=FunctionInput(Preview=Constant()) {bUsePreviewValueAsDefault=true, inputName=CornerRoundness, inputType=FunctionInput_Vector4, sortPriority=10}) {name=CornerRoundness, nodeColor=<rgba 0 0.4583 0.0978 1>}
  NamedRerouteDeclaration(None=FunctionInput(Preview=Constant() {r=2}) {bUsePreviewValueAsDefault=true, inputName=ProgressSizeY, inputType=FunctionInput_Scalar, sortPriority=12}) {name=ProgressSizeY, nodeColor=<rgba 0 0.4583 0.0978 1>}
  NamedRerouteDeclaration(None=FunctionInput(Preview=Constant() {r=1}) {bUsePreviewValueAsDefault=true, inputName=ProgressParam, inputType=FunctionInput_Scalar}) {name=ProgressParam, nodeColor=<rgba 0 0.4583 0.0978 1>}
  NamedRerouteDeclaration(None=FunctionInput(Preview=Constant() {r=1}) {bUsePreviewValueAsDefault=true, inputName=ScaleWidthParam, inputType=FunctionInput_Scalar, sortPriority=2}) {name=ScalarWidthParam, nodeColor=<rgba 0 0.4583 0.0978 1>}
  NamedRerouteDeclaration(None=FunctionInput(Preview=StaticBool()) {bUsePreviewValueAsDefault=true, inputName=Reverse, inputType=FunctionInput_StaticBool, sortPriority=1}) {name=Reverse, nodeColor=<rgba 0.807 0.0513 0.0331 1>}
  NamedRerouteDeclaration(None=FunctionInput(Preview=TextureCoordinate()) {bUsePreviewValueAsDefault=true, inputName=UVs, inputType=FunctionInput_Vector2, sortPriority=-1}) {name=UVs, nodeColor=<rgba 0.5208 0.0298 0.2499 1>}
 ]
}
