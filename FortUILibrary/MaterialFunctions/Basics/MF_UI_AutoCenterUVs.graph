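# UI material function in the Juno UI library that shifts shape UVs to center a box defined by left, top, right and bottom sizes.
# Per axis, the offset is the larger opposing size minus half their sum.
# An IsGreater check drives a lerp that adds or subtracts this offset, and both axes recombine as Centered UVs.
# UVs default to texture coordinates remapped around zero, and Normalized Size defaults to zero.
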MaterialFunction /FortUILibrary/MaterialFunctions/Basics/MF_UI_AutoCenterUVs
graph {
 properties {
  bExposeToLibrary=true
  description="This function takes the current box size as a Vector4 (left, top, right, bottom) and adjusts the UVs so that the shape is automatically centered within its local UV space."
  previewMaterialDomain=MD_UI
  libraryCategoriesText [Juno UI]
 }
 nodes [
  n1 = ComponentMask(None=NamedRerouteUsage()) {r=true}
  n2 = NamedRerouteUsage()
  n3 = NamedRerouteUsage()
  n4 = Add(A=@n2, B=@n3)
  n5 = Subtract(
   A=NamedRerouteDeclaration(None=Max(A=@n2, B=@n3)) {name=Max.x, nodeColor=<rgba 0 0 0 1>}
   B=Multiply(A=NamedRerouteDeclaration(None=@n4) {name=Distance.x, nodeColor=<rgba 0 0 0 1>}) {constB=0.5}
  )
  n6 = NamedRerouteUsage()
  n7 = NamedRerouteUsage()
  n8 = LinearInterpolate(
   A=Subtract(A=@n1, B=@n5)
   B=Add(A=@n1, B=@n5)
   Alpha=NamedRerouteDeclaration(None=MaterialFunctionCall!(A=@n6, B=@n7) {materialFunction=&/MF_UI_BooleanOperations}.IsGreater) {name=Size Left is Greater than Size Right, nodeColor=<rgba 1 0 0.1402 1>}
  )
  n9 = ComponentMask(None=NamedRerouteUsage()) {g=true}
  n10 = NamedRerouteUsage()
  n11 = NamedRerouteUsage()
  n12 = Add(A=@n10, B=@n11)
  n13 = Subtract(
   A=NamedRerouteDeclaration(None=Max(A=@n10, B=@n11)) {name=Max.y, nodeColor=<rgba 0 0 0 1>}
   B=Multiply(A=NamedRerouteDeclaration(None=@n12) {name=Distance.y, nodeColor=<rgba 0 0 0 1>}) {constB=0.5}
  )
  n14 = NamedRerouteUsage()
  n15 = NamedRerouteUsage()
  n16 = LinearInterpolate(
   A=Subtract(A=@n9, B=@n13)
   B=Add(A=@n9, B=@n13)
   Alpha=NamedRerouteDeclaration(None=MaterialFunctionCall(A=@n14, B=@n15) {materialFunction=&MF_UI_BooleanOperations}.IsGreater) {name=Size Top is Greater than Size Bottom, nodeColor=<rgba 1 0 0.1402 1>}
  )
  FunctionOutput(None=AppendVector(
    A=NamedRerouteDeclaration(None=@n8) {name=Size Center.x, nodeColor=<rgba 0 0 0 1>}
    B=NamedRerouteDeclaration(None=@n16) {name=Size Center.y, nodeColor=<rgba 0 0 0 1>}
   )) {outputName=Centered UVs}
  MaterialFunctionCall(
   X=Constant() {desc="R (left)", r=0.2}
   Y=Constant() {desc="G (top)", r=0.1}
   Z=Constant() {desc="B (right)", r=0.25}
   A=Constant() {desc="A (bottom)", r=0.35}
  ) {materialFunction=&MakeFloat4}
  n17 = TextureCoordinate()
  NamedRerouteDeclaration(None=FunctionInput(Preview=ConstantBiasScale(None=@n17) {bias=-0.5, desc=Try UVs Centered by default, scale=2}) {bUsePreviewValueAsDefault=true, inputName=Shape UVs, inputType=FunctionInput_Vector2}) {name=UVs, nodeColor=<rgba 0 0 0 1>}
  n18 = FunctionInput() {
   bUsePreviewValueAsDefault=true
   description="Normalized Size value (0 to 1)"
   inputName=Normalized Size
   inputType=FunctionInput_Vector4
   sortPriority=1
   previewValue=<0 0 0 0>
  }
  n19 = MaterialFunctionCall(Float4=@n18) {materialFunction=&BreakOutFloat4Components}
  NamedRerouteDeclaration(None=@n19.R) {name=Size Left.x, nodeColor=<rgba 0 0 0 1>}
  NamedRerouteDeclaration(None=@n19.G) {name=Size Top.y, nodeColor=<rgba 0 0 0 1>}
  NamedRerouteDeclaration(None=@n19.B) {name=Size Right.x, nodeColor=<rgba 0 0 0 1>}
  NamedRerouteDeclaration(None=@n19.A) {name=Size Bottom.y, nodeColor=<rgba 0 0 0 1>}
 ]
}
