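# UI material function that outputs a raw, non-antialiased box signed distance field, sized independently per side (left, top, right, bottom).
# Centered UVs are offset by the side sizes, then feed a max, length and min distance calculation.
# Per-corner radii come from isolating quadrants on UVs re-centered on the shape, so asymmetric boxes and UV transforms carry through.
# An optional aspect-ratio switch, derived from pixel size, keeps corners undistorted on non-square widgets.
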MaterialFunction /FortUILibrary/MaterialFunctions/Shapes/SDFs/MF_SDF_Box_V4
graph {
 properties {
  bExposeToLibrary=true
  description="Signed distance field for a box using per-side sizing from a Vector4 (left, top, right, bottom), with optional rounded corners."
  previewMaterialDomain=MD_UI
  libraryCategoriesText [SDFs]
 }
 nodes [
  Constant4Vector() {constant=<rgba 0.5 0.25 0.1 0.14>}
  Constant4Vector() {constant=<rgba 0.5 0.8 0.4 0.2>}
  Constant() {r=0.8}
  Constant()
  n1 = NamedRerouteUsage()
  n2 = FunctionInput(Preview=TextureCoordinate()) {
   bUsePreviewValueAsDefault=true
   description="Controls the UV mapping used to sample and shape the geometry. Use custom UVs to stretch, deform, or reposition the shape within the widget space."
   inputName=UVs
   inputType=FunctionInput_Vector2
  }
  n3 = NamedRerouteDeclaration(None=ConstantBiasScale(None=@n2) {bias=-0.5, scale=2}) {name=Centered UVs, nodeColor=<rgba 0 0 0 1>}
  n4 = Multiply(A=Add(A=@n3, B=ComponentMask(None=@n1) {desc="R:Left.x\r\nG:Top.y", g=true, r=true})) {constB=-1}
  n5 = Subtract(A=@n3, B=ComponentMask(None=@n1) {a=true, b=true, desc="B: Right.x\r\nA: Bottom.y"})
  n6 = NamedRerouteUsage()
  n7 = Add(A=Multiply(A=Max(A=@n4, B=@n5), B=NamedRerouteUsage()) {constB=2}, B=@n6)
  n8 = Length(None=Max(A=@n7) {constB=0}) {is=class MaterialExpressionLength}
  n9 = MaterialFunctionCall(Float2=@n7) {materialFunction=&BreakOutFloat2Components}
  n10 = Min(A=Max(A=@n9.R, B=@n9.R) {constB=0}) {constB=0}
  FunctionOutput(None=Subtract(A=Add(A=@n8, B=@n10), B=@n6)) {
   description="Outputs the raw signed distance field (SDF) of the shape. No anti-aliasing applied. Used internally to generate the Fill output."
   outputName=SDF
   sortPriority=3
  }
  n11 = NamedRerouteDeclaration(None=MaterialFunctionCall!(
    Shape UVs=NamedRerouteUsage() {desc="Necesary to apply modification like rotation, slate etc. comming from the orginal UVs to the Rounded corners Quadrant UVs"}
    Normalized Size=NamedRerouteUsage()
   ) {materialFunction=/FortUILibrary/MaterialFunctions/Basics/MF_UI_AutoCenterUVs}.Centered UVs) {
   name=UVs Centered on Shape
   desc="Any modifications applied on the shape UVs are appled to these UVs Centered on Shape \r\nbecause they derived from the Shape Centered UVs.\r\nIt is mandatory to apply the effects (rotation etc) to the rounded corners."
   nodeColor=<rgba 1 0.0197 0 1>
  }
  n12 = Constant() {
   desc="The Constant Bias Scale node must be ignored within the Isolate Quads function. \r\nIt must be applied before the function; otherwise, it causes issues with thin shapes."
   r=1
  }
  n13 = FunctionInput() {
   bUsePreviewValueAsDefault=true
   description="R= Top Left, G= Bottom Left, B= Top Right, A= Bottom Right. The radius (usually between '0.0' and '0.5') of the rounded corners. Use a Float (Scalar) for uniform roundness, or a Float4 to set a different radius for each corner."
   inputName=Corner Roundness
   inputType=FunctionInput_Vector4
   sortPriority=7
   previewValue=<0 0 0 0>
  }
  NamedRerouteDeclaration(None=MaterialFunctionCall!(UVs=@n11, Use Raw UVs=@n12, 4-Quad Values=@n13) {materialFunction=/FortUILibrary/MaterialFunctions/Basics/MF_UI_IsolateQuads}.Isolated 4-Quad) {name=Corner Roundness, nodeColor=<rgba 0 0 0 1>}
  n14 = FunctionInput(Preview=MaterialFunctionCall() {materialFunction=&GetUserInterfaceUV}.Pixel Size) {
   bUsePreviewValueAsDefault=true
   description=Pixel Size Reference
   inputName=Pixel Size
   inputType=FunctionInput_Vector2
   sortPriority=100
   previewValue=<0.5 0.5 0 1>
  }
  n15 = FunctionInput() {
   bUsePreviewValueAsDefault=true
   description="0 = False, 1 = True (default). Preserves the proportions of rounded corners to avoid distortion when the widget is non-square. Has no visible effect if corner roundness is 0."
   inputName=Preserve Roundness
   inputType=FunctionInput_Scalar
   sortPriority=8
   previewValue=<1 0 0 0>
  }
  NamedRerouteDeclaration(None=LinearInterpolate(
    A=Constant() {r=1}
    B=MaterialFunctionCall!(Dimensions=@n14) {materialFunction=/FortUILibrary/MaterialFunctions/Basics/MF_UI_GetAspectRatio}.Max Ratio
    Alpha=Saturate(None=@n15)
   )) {name=Ratio Switch, nodeColor=<rgba 0 0.117 0.102 1>}
  n16 = FunctionInput() {
   bUsePreviewValueAsDefault=true
   description="Controls the size from the shape’s center. Use a float (scalar) for uniform sizing, or a float4 to define individual size (0 to 1) per side. Values are interpreted as a ratio (0 to 1) relative to the widget container."
   inputName=Size
   inputType=FunctionInput_Vector4
   sortPriority=3
   previewValue=<0.5 0.5 0.5 0.5>
  }
  NamedRerouteDeclaration(None=@n16) {name=Size, nodeColor=<rgba 0 0 0 1>}
 ]
}
