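# MF_SDF_Circle is a UI-domain material function, exposed in the SDFs library category, that outputs a signed distance field for a circle.
# It remaps the UVs input, defaulting to texture coordinates, from 0–1 to -1–1 so the circle centers on mid-UV.
# The vector's length minus the Radius input gives the SDF, negative inside and positive outside, measured in that remapped space.
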MaterialFunction /FortUILibrary/MaterialFunctions/Shapes/SDFs/MF_SDF_Circle
graph {
 properties {bExposeToLibrary=true, description=Signed distance field for a circle., previewMaterialDomain=MD_UI, libraryCategoriesText [SDFs]}
 nodes [
  n1 = FunctionInput(Preview=TextureCoordinate()) {
   bUsePreviewValueAsDefault=true
   description="Coordinates of the distance field (0,0 is origin)."
   inputName=UVs
   inputType=FunctionInput_Vector2
  }
  n2 = ConstantBiasScale(None=@n1) {bias=-0.5, scale=2}
  n3 = FunctionInput() {
   bUsePreviewValueAsDefault=true
   description=The radius of the circle.
   inputName=Radius
   inputType=FunctionInput_Scalar
   sortPriority=1
   previewValue=<0.5 0 0 1>
  }
  FunctionOutput(None=Subtract(A=Length(None=@n2) {is=class MaterialExpressionLength}, B=@n3)) {description=The resulting circle SDF., outputName=SDF}
 ]
}
