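# UI material function in the SDFs library category that outputs a signed distance field for an isosceles trapezoid with adjustable TopWidth, BottomWidth and Height.
# It remaps UVs around the centre and mirrors X with Abs to evaluate one half.
# It takes the nearer squared distance to the cap edges or the clamped projection onto the slanted side.
# The larger of two edge signs makes the square-rooted result negative inside.
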MaterialFunction /FortUILibrary/MaterialFunctions/Shapes/SDFs/MF_SDF_IsoscelesTrapezoid
graph {
 properties {
  bExposeToLibrary=true
  description=Signed distance field for an isosceles trapezoid.
  previewMaterialDomain=MD_UI
  libraryCategoriesText [SDFs]
 }
 nodes [
  n1 = FunctionInput(Preview=TextureCoordinate()) {
   bUsePreviewValueAsDefault=true
   description="Coordinates of the distance field (0,0 is origin)."
   inputName=UVs
   inputType=FunctionInput_Vector2
  }
  n2 = ConstantBiasScale(None=@n1) {bias=-0.5, scale=2}
  n3 = Abs(None=ComponentMask(None=@n2) {r=true})
  n4 = ComponentMask(None=@n2) {g=true}
  n5 = AppendVector(A=@n3, B=@n4)
  n6 = FunctionInput() {
   bUsePreviewValueAsDefault=true
   description=The width of the bottom edge of the trapezoid.
   inputName=BottomWidth
   inputType=FunctionInput_Scalar
   sortPriority=2
   previewValue=<0.7 0 0 1>
  }
  n7 = FunctionInput() {
   bUsePreviewValueAsDefault=true
   description=The height of the trapezoid.
   inputName=Height
   inputType=FunctionInput_Scalar
   sortPriority=3
   previewValue=<0.5 0 0 1>
  }
  n8 = AppendVector(A=@n6, B=@n7) {desc=k1}
  n9 = FunctionInput() {
   bUsePreviewValueAsDefault=true
   description=The width of the top edge of the trapezoid.
   inputName=TopWidth
   inputType=FunctionInput_Scalar
   sortPriority=1
   previewValue=<0.3 0 0 1>
  }
  n10 = AppendVector(A=Subtract(A=@n6, B=@n9), B=Multiply(A=@n7) {constB=2}) {desc=k2}
  n11 = Divide(A=DotProduct(A=Subtract(A=@n8, B=@n5), B=@n10), B=DotProduct(A=@n10, B=@n10))
  n12 = Add(A=Subtract(A=@n5, B=@n8), B=Multiply(A=Saturate(None=@n11), B=@n10)) {desc=cb}
  n13 = Sign(None=ComponentMask(None=@n12) {r=true})
  n14 = If(A=@n4, "A > B"=@n6, "A == B"=@n6, "A < B"=@n9)
  n15 = AppendVector(A=Subtract(A=@n3, B=Min(A=@n3, B=@n14)), B=Subtract(A=Abs(None=@n4), B=@n7)) {desc=ca}
  n16 = Sign(None=ComponentMask(None=@n15) {g=true})
  n17 = Min(A=DotProduct(A=@n12, B=@n12), B=DotProduct(A=@n15, B=@n15))
  FunctionOutput(None=Multiply(A=Max(A=@n13, B=@n16), B=SquareRoot(None=@n17))) {description=The resulting trapezoid SDF., outputName=SDF}
 ]
}
