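# Material function in the UI library that outputs a distance field for a line segment between inputs PointA and PointB.
# It remaps UVs from 0..1 to -1..1, so the origin sits at the center and the points use that space.
# It projects the coordinate onto the segment, saturates the factor to clamp at the endpoints, and returns the length to that closest point.
# With no radius subtracted, the distance is never negative.
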MaterialFunction /FortUILibrary/MaterialFunctions/Shapes/SDFs/MF_SDF_LineSegment
graph {
 properties {bExposeToLibrary=true, description=Signed distance field for a line segment., previewMaterialDomain=MD_UI, libraryCategoriesText [SDFs]}
 nodes [
  n1 = FunctionInput(Preview=TextureCoordinate()) {
   bUsePreviewValueAsDefault=true
   description="Coordinates of the distance field (0,0 is origin)."
   inputName=UVs
   inputType=FunctionInput_Vector2
  }
  n2 = FunctionInput() {
   bUsePreviewValueAsDefault=true
   description=The coordinates of the first point.
   inputName=PointA
   inputType=FunctionInput_Vector2
   sortPriority=1
   previewValue=<-0.5 -0.5 0 1>
  }
  n3 = Subtract(A=ConstantBiasScale(None=@n1) {bias=-0.5, scale=2}, B=@n2) {desc=pa}
  n4 = FunctionInput() {
   bUsePreviewValueAsDefault=true
   description=The coordinates of the second point.
   inputName=PointB
   inputType=FunctionInput_Vector2
   sortPriority=2
   previewValue=<0.5 0.5 0 1>
  }
  n5 = Subtract(A=@n4, B=@n2) {desc=ba}
  n6 = DotProduct(A=@n3, B=@n5)
  n7 = DotProduct(A=@n5, B=@n5)
  n8 = Multiply(A=Saturate(None=Divide(A=@n6, B=@n7)) {desc=h}, B=@n5)
  FunctionOutput(None=Length(None=Subtract(A=@n3, B=@n8)) {is=class MaterialExpressionLength}) {description=The resulting line segment SDF., outputName=SDF}
 ]
}
