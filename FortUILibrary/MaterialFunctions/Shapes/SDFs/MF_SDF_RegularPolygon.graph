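# UI material function that returns a signed distance field for a regular polygon with variable Sides and Radius, with InnerAngle turning it into star or flower shapes.
# It remaps the input UVs around a centre, then folds polar angle by modulo and absolute sine into one mirrored sector.
# It projects onto that sector's edge with a clamped dot product, outputting the distance signed by the X component.
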MaterialFunction /FortUILibrary/MaterialFunctions/Shapes/SDFs/MF_SDF_RegularPolygon
graph {
 properties {
  bExposeToLibrary=true
  description=Signed distance field for a regular polygon with variable sides.
  previewMaterialDomain=MD_UI
  libraryCategoriesText [SDFs]
 }
 nodes [
  n1 = FunctionInput(Preview=TextureCoordinate()) {
   bUsePreviewValueAsDefault=true
   description="Coordinates of the distance field (0,0 is origin)."
   inputName=UVs
   inputType=FunctionInput_Vector2
  }
  n2 = ConstantBiasScale(None=@n1) {bias=-0.5, scale=2}
  n3 = Abs(None=ComponentMask(None=@n2) {r=true})
  n4 = Multiply(A=ComponentMask(None=@n2) {g=true}) {constB=-1}
  n5 = Constant() {desc=pi, r=3.1416}
  n6 = FunctionInput() {
   bUsePreviewValueAsDefault=true
   description=The number of sides/points on the polygon.
   inputName=Sides
   inputType=FunctionInput_Scalar
   sortPriority=2
   previewValue=<5 0 0 1>
  }
  n7 = Divide(A=@n5, B=@n6) {desc=an}
  n8 = Subtract(A=Modulo!(A=Arctangent2Fast(Y=@n3, X=@n4), B=Multiply(A=@n7) {constB=2}), B=@n7) {desc=bn}
  n9 = Cosine(None=@n8) {period=6.2832}
  n10 = Abs(None=Sine(None=@n8) {period=6.2832})
  n11 = AppendVector(A=Cosine(None=@n7) {period=6.2832}, B=Sine(None=@n7) {period=6.2832}) {desc=acs}
  n12 = FunctionInput() {
   bUsePreviewValueAsDefault=true
   description=The radius of the polygon.
   inputName=Radius
   inputType=FunctionInput_Scalar
   sortPriority=1
   previewValue=<0.5 0 0 1>
  }
  n13 = Subtract(A=Multiply(A=Length(None=@n2) {is=class MaterialExpressionLength}, B=AppendVector(A=@n9, B=@n10)), B=Multiply(A=@n11, B=@n12) {constB=2})
  n14 = FunctionInput() {
   bUsePreviewValueAsDefault=true
   description="Add inner angles to the polygon (eg. to make a star or flower shape)."
   inputName=InnerAngle
   inputType=FunctionInput_Scalar
   sortPriority=3
  }
  n15 = Divide(A=@n5, B=Add(A=@n14) {constB=2})
  n16 = AppendVector(A=Cosine(None=@n15) {period=6.2832}, B=Sine(None=@n15) {period=6.2832}) {desc=ecs}
  n17 = Multiply(A=DotProduct(A=@n13, B=@n16)) {constB=-1}
  n18 = Multiply(A=@n12, B=Divide(A=ComponentMask(None=@n11) {g=true}, B=ComponentMask(None=@n16) {g=true})) {constA=2}
  n19 = Add(A=@n13, B=Multiply(A=Clamp(None=@n17, Max=@n18), B=@n16))
  n20 = ComponentMask(None=@n19) {r=true}
  FunctionOutput(None=Multiply(A=Length(None=@n19), B=Sign(None=@n20))) {description=The resulting polygon SDF., outputName=SDF}
 ]
}
